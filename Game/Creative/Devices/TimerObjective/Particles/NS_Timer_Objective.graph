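# Niagara system for the Creative Timer Objective device: a short colored flare with a light flash.
# A single Flare emitter burst-spawns brief particles whose sprite size wobbles by curve while brightness rises and fades, tinted by User.Color.
# Each particle feeds two sprite renderers and a light renderer, with separate corona size and camera-offset attributes set per frame.
# User.LightOpacity scales the light's curve-driven intensity.
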NiagaraSystem /Game/Creative/Devices/TimerObjective/Particles/NS_Timer_Objective
dependencies {
 usedDataInterfaces [&NiagaraDataInterfaceCurve, &NiagaraDataInterfaceVector2DCurve]
 usedDynamicInputs []
 usedModules [&SystemState]
}
emitters {Flare {
  emitterUpdateScript [
   &EmitterState()
   &SpawnBurst_Instantaneous(Use Spawn Probability=false, UseLoopCountLimit=false)
  ]
  particleSpawnScript [
   &InitializeParticle(Lifetime=0.35, Sprite Size=<500 250>, Sprite Size Mode=::NewEnumerator4, UsePositionOffset=false, Write Material Random=true)
   &Color(Color=&MakeLinearColorFromVectorAndFloat("Vector (RGB)"=<0.0883 0.7639 0.1109>, "Float (Alpha)"=&ClampFloat(Float=1)))
  ]
  particleUpdateScript [
   &ParticleState()
   &ScaleSpriteSize(
    Scale Factor=&Vector2DFromCurve(
     Vector2Curve=DataInterface() {propertyValues {
       XCurve {keys [
         {arriveTangent=-0.5986, time=0, value=1}
         {arriveTangent=0.9634, leaveTangent=0.9634, time=0.3322, value=0.4014}
         {arriveTangent=-0.373, leaveTangent=-0.373, time=0.6667, value=1.3648}
         <at 0.9989 0.9918>
        ]}
       YCurve {keys [
         {arriveTangent=0.4337, time=0, value=1}
         {arriveTangent=-0.9277, leaveTangent=-0.9277, time=0.3322, value=1.4337}
         {arriveTangent=0.4916, leaveTangent=0.4916, time=0.6667, value=0.506}
         <at 0.9989 0.9976>
        ]}
      }}
     ScaleCurve=<1 1>
    )
    Scale Sprite Size Mode=::NewEnumerator2
   )
   &ScaleColor(
    Color Value To Scale=$Particles.Color
    Scale RGB=&Multiply_VectorByFloat(
     Vector=&MakeVectorFromLinearColorRGB(Color=$User.Color)
     Float=&FloatFromCurve(FloatCurve=DataInterface() {propertyValues {Curve {keys [<at 0 0 RCIM_Cubic>,<at 0.2 1 RCIM_Cubic>,<at 1 0 RCIM_Cubic>]}}})
    )
    ScaleA=true
    ScaleRGB=true
   )
   &CameraOffset(Camera Offset Amount=25)
   /SetVariables(
    Particles.CoronaCameraOffset=250
    Particles.CoronaSize:Vector2f=&Multiply_Vector2D_ByFloat(
     Float=&FloatFromCurve(FloatCurve=DataInterface() {propertyValues {Curve {keys [<at 0 1 RCIM_Cubic>,<at 0.5 1.2 RCIM_Cubic>,<at 1 1 RCIM_Cubic>]}}})
     Vector2D=<120 120>
    )
   ) {isSetParameters=true}
   &FN_Light_Attributes(
    Color=$Particles.Color
    Create Color Attribute=true
    Create Exponent Attribute=true
    Create Radius Attribute=true
    Exponent=4
    Radius=300
    Scale Color=&FloatFromCurve(
     FloatCurve=DataInterface() {propertyValues {Curve {keys [
         {interpMode=RCIM_Cubic, tangentMode=RCTM_User, time=0, value=0}
         {interpMode=RCIM_Cubic, tangentMode=RCTM_User, time=0.5, value=1}
         {tangentMode=RCTM_User, time=1, value=0}
        ]}}}
     Scale Curve=&Multiply_Float(A=20, B=$User.LightOpacity)
    )
   )
  ]
  rendererClasses [&NiagaraSpriteRendererProperties, &NiagaraLightRendererProperties]
  renderers [
   {rendererClass=&NiagaraSpriteRendererProperties}
   {rendererClass=&NiagaraSpriteRendererProperties, rendererIndex=1}
   {rendererClass=&NiagaraLightRendererProperties, rendererIndex=2}
  ]
 }}
summary {
 systemName=NS_Timer_Objective
 emitters [{emitterName=Flare, rendererClasses [&NiagaraSpriteRendererProperties, &NiagaraLightRendererProperties]}]
 userVariables [
  User.Color {type=LinearColor, description="", defaultValue {struct=LinearColor, value=<rgba 1 1 1 1>}}
  User.LightOpacity {type=NiagaraFloat, defaultValue=1, description=""}
 ]
}
